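# Post-adjusts the F_MED_Faux female cosmetic body, copying the attached parent's pose and applying corrective bone offsets.
# The AnimGraph constrains and offsets the right upper-arm fix bone, then blends additive thigh-fix and glute transforms by variable alphas.
# Each update derives alphas from bone twist, refreshing thigh-forward only while crouching and damping glute values above a speed threshold.
# The computed arm-raise alpha drives no node.
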AnimBlueprint /BRCosmetics/Characters/Player/Female/Medium/Bodies/F_MED_Faux/Meshes/F_MED_Faux_AnimBP
target_skeleton=&/F_MED_Faux_Skeleton
graphs [
 AnimGraph {nodes [
   "Transform (Modify) Bone - Bone: thigh_l"(Alpha=1, Translation="0,0,0", Rotation="0,-90,0", Scale="1,1,1") {
    rotationMode=BMM_Additive
    rotationSpace=BCS_BoneSpace
    alphaBoolBlend {blendOption=Linear}
    boneToModify {boneName=thigh_l}
    rotation {yaw=-90}
   }
   n1 = "Transform (Modify) Bone - Bone: clavicle_r"(Alpha=1, Translation="0,0,0", Rotation="28.000,0.000,0.000", Scale="1,1,1") {
    rotationMode=BMM_Additive
    rotationSpace=BCS_BoneSpace
    translationMode=BMM_Additive
    alphaBoolBlend {blendOption=Linear}
    boneToModify {boneName=clavicle_r}
    rotation {pitch=28}
   }
   "Transform (Modify) Bone - Bone: upperarm_r"(ComponentPose=@n1, Alpha=1, Translation="0,0,0", Rotation="16.512,93.803,-4.819", Scale="1,1,1") {
    rotationMode=BMM_Additive
    rotationSpace=BCS_BoneSpace
    translationMode=BMM_Additive
    alphaBoolBlend {blendOption=Linear}
    boneToModify {boneName=upperarm_r}
    rotation=<rot 16.512 93.803 -4.819>
   }
   n2 = "Constraint - upperarm_fix_r (P)"(
    ComponentPose=Local To Component(LocalPose=Copy Pose From Mesh() {bCopyCurves=true, bUseAttachedParent=true})
    Alpha=1
    ConstraintWeights_0=1
   ) {
    alphaBoolBlend {blendOption=Linear}
    boneToModify {boneName=upperarm_fix_r}
    constraintSetup [{targetBone {boneName=upperarm_twist_01_r}}]
    constraintWeights [1]
   }
   n3 = "Transform (Modify) Bone - Bone: upperarm_fix_r"(ComponentPose=@n2, Alpha=1, Translation="-2,-2.094,-2", Rotation="9.758,1.981,22.967", Scale="0.95,0.95,0.95") {
    rotationMode=BMM_Additive
    rotationSpace=BCS_BoneSpace
    scaleMode=BMM_Additive
    scaleSpace=BCS_BoneSpace
    translationMode=BMM_Additive
    translationSpace=BCS_BoneSpace
    alphaBoolBlend {blendOption=Linear}
    boneToModify {boneName=upperarm_fix_r}
    rotation=<rot 9.758 1.981 22.9669>
    scale=<0.95 0.95 0.95>
    translation=<-2 -2.094 -2>
   }
   n4 = "Transform (Modify) Bone - Bone: thigh_fix_l"(ComponentPose=@n3, Alpha=Get thigh_fwd_l() -> (thigh_fwd_l:double), Translation="-7,0.077,0.000", Rotation="0.000,40,0.000", Scale="1,1,1") {
    rotationMode=BMM_Additive
    rotationSpace=BCS_BoneSpace
    scaleMode=BMM_Additive
    scaleSpace=BCS_BoneSpace
    translationMode=BMM_Additive
    translationSpace=BCS_BoneSpace
    alphaBoolBlend {blendOption=Linear}
    alphaScaleBiasClamp {bInterpResult=true, interpSpeedDecreasing=25, interpSpeedIncreasing=25}
    boneToModify {boneName=thigh_fix_l}
    rotation {yaw=40}
    translation=<-7 0.077>
   }
   n5 = "Transform (Modify) Bone - Bone: thigh_fix_r"(ComponentPose=@n4, Alpha=Get thigh_fwd_r() -> (thigh_fwd_r:double), Translation="7,0.077,0.000", Rotation="0.000,40,0.000", Scale="1,1,1") {
    rotationMode=BMM_Additive
    rotationSpace=BCS_BoneSpace
    scaleMode=BMM_Additive
    scaleSpace=BCS_BoneSpace
    translationMode=BMM_Additive
    translationSpace=BCS_BoneSpace
    alphaBoolBlend {blendOption=Linear}
    alphaScaleBiasClamp {bInterpResult=true, interpSpeedDecreasing=25, interpSpeedIncreasing=25}
    boneToModify {boneName=thigh_fix_r}
    rotation {yaw=40}
    translation=<7 0.077>
   }
   n6 = "Transform (Modify) Bone - Bone: dyn_glute_l"(ComponentPose=@n5, Alpha=Get thigh_rot_l(), Translation="-1.000,0.25,2.5", Rotation="0.000,0.000,-30", Scale="1.25,1.25,1.25") {
    rotationMode=BMM_Additive
    rotationSpace=BCS_BoneSpace
    scaleMode=BMM_Additive
    scaleSpace=BCS_BoneSpace
    translationMode=BMM_Additive
    translationSpace=BCS_BoneSpace
    alphaBoolBlend {blendOption=Linear}
    boneToModify {boneName=dyn_glute_l}
    rotation {roll=-30}
    scale=<1.25 1.25 1.25>
    translation=<-1 0.25 2.5>
   }
   n7 = "Transform (Modify) Bone - Bone: dyn_glute_r"(ComponentPose=@n6, Alpha=Get thigh_rot_r(), Translation="-1,0.25,2.5", Rotation="0,0,-30", Scale="1.25,1.25,1.25") {
    rotationMode=BMM_Additive
    rotationSpace=BCS_BoneSpace
    scaleMode=BMM_Additive
    scaleSpace=BCS_BoneSpace
    translationMode=BMM_Additive
    translationSpace=BCS_BoneSpace
    alphaBoolBlend {blendOption=Linear}
    boneToModify {boneName=dyn_glute_r}
    rotation {roll=-30}
    scale=<1.25 1.25 1.25>
    translation=<-1 0.25 2.5>
   }
   Output Pose(Result=Component To Local(ComponentPose=@n7)) {layerGroup=None}
   Get arm_up_r() -> (arm_up_r:double)
  ]}
 EventGraph {nodes [
   Event BlueprintUpdateAnimation()
   > Get!(execute:Exec) -> (then:Exec, else:Exec, Skeletal_Mesh:Skeletal Mesh Component Object Reference)
   > n1 = Sequence!(execute) -> (then_0, then_1, then_2:Exec, then_3:Exec)
   > Set arm_up_r(arm_up_r:double=GetAlphaFromReferenceTransformDelta(
     SkeletalMeshComponent=Get Skeletal_Mesh() -> (Skeletal_Mesh:Skeletal Mesh Component Object Reference)
     BoneName=upperarm_r
     BaseName=clavicle_r
     DeltaComponentType=QuaternionTwist
     InputRange=(X=-10,Y=-100)
     TwistAxis="0,0,1"
    )) -> (Output_Get:double)
   Branch(execute=@n1.then_1, Condition=Get bIsCrouching())
   > Set thigh_fwd_l(thigh_fwd_l:double=GetAlphaFromReferenceTransformDelta(
     SkeletalMeshComponent=Get Skeletal_Mesh()
     BoneName=thigh_l
     BaseName=pelvis
     DeltaComponentType=QuaternionTwist
     InputRange=(X=0,Y=-90)
     TwistAxis="0,0,1"
    )) -> (Output_Get:double)
   > Set thigh_fwd_r(thigh_fwd_r:double=GetAlphaFromReferenceTransformDelta(
     SkeletalMeshComponent=Get Skeletal_Mesh()
     BoneName=thigh_r
     BaseName=pelvis
     DeltaComponentType=QuaternionTwist
     InputRange=(X=0,Y=90)
     TwistAxis="0,0,1"
    )) -> (Output_Get:double)
   n2 = Get Skeletal_Mesh()
   n3 = GetAlphaFromReferenceTransformDelta(SkeletalMeshComponent=@n2, BoneName=thigh_l, BaseName=pelvis, DeltaComponentType=QuaternionTwist, InputRange=(X=20,Y=100), TwistAxis="0,0,1")
   Set thigh_rot_l(execute=@n1.then_2, thigh_rot_l=@n3)
   n4 = GetAlphaFromReferenceTransformDelta(SkeletalMeshComponent=@n2, BoneName=thigh_r, BaseName=pelvis, DeltaComponentType=QuaternionTwist, InputRange=(X=20,Y=100), TwistAxis="0,0,1")
   > Set thigh_rot_r(thigh_rot_r=@n4)
   > Branch(Condition="float > float"(A=Get PawnSpeed2D(), B=100) {is=title K2Node_CallFunction})
   > Set thigh_rot_l(thigh_rot_l=float * float(A=@n3, B=0.35))
   > Set thigh_rot_r(thigh_rot_r=float * float(A=@n4, B=0.35))
   Event BlueprintInitializeAnimation()
   > Set Skeletal_Mesh(Skeletal_Mesh:Skeletal Mesh Component Object Reference=GetPartSkeletalMeshComponent()) -> (Output_Get:Skeletal Mesh Component Object Reference)
  ]}
]
variables [
 arm_up_r {type=number}
 Skeletal_Mesh {type=SkeletalMeshComponent}
 thigh_fwd_l {type=number}
 thigh_fwd_r {type=number}
 thigh_rot_l {type=number}
 thigh_rot_r {type=number}
]
