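# Physics asset for the F_MED_Faux female medium face accessory, setting up ragdoll-style bodies for its dynamic strand bones.
# Small capsules cover the head and six strand bones.
# A chain runs from head through parent strand and strand, then branches to top and bottom tips, while separate left and right strands attach directly to the head.
# Narrow swing limits and very tight twist limits restrict each joint to subtle motion.
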PhysicsAsset /BRCosmetics/Characters/Player/Female/Medium/Bodies/F_MED_Faux/Meshes/Parts/F_MED_Faux_FaceAcc_Physics
bodies {
 dyn_parent_strand [{shapeType=Capsule, length=1, radius=1, center=<1 0 0>, rotation=<rot 90 0 0>}]
 dyn_strand [{shapeType=Capsule, length=2.25, radius=2, center=<0.1 0 0>, rotation=<rot 90 0 0>}]
 dyn_strand_01_l [{shapeType=Capsule, length=1, radius=2, center=<0.1 0 0>, rotation=<rot 90 0 0>}]
 dyn_strand_01_r [{shapeType=Capsule, length=1.75, radius=2, center=<0.3 0 0>, rotation=<rot 90 0 0>}]
 dyn_strand_bot_01 [{shapeType=Capsule, length=1, radius=0.75, center=<0.35 0 0>, rotation=<rot 90 0 0>}]
 dyn_strand_top_01 [{shapeType=Capsule, length=1.75, radius=1.25, center=<0.25 0 0>, rotation=<rot 90 0 0>}]
 head [{shapeType=Capsule, length=1, radius=1}]
}
constraints [
 {bone1Name=dyn_parent_strand, bone2Name=head, swing1LimitDegrees=15, swing2LimitDegrees=15, twistLimitDegrees=3}
 {bone1Name=dyn_strand, bone2Name=dyn_parent_strand, swing1LimitDegrees=15, swing2LimitDegrees=15, twistLimitDegrees=3}
 {bone1Name=dyn_strand_top_01, bone2Name=dyn_strand, swing1LimitDegrees=15, swing2LimitDegrees=15, twistLimitDegrees=3}
 {bone1Name=dyn_strand_bot_01, bone2Name=dyn_strand, swing1LimitDegrees=20, swing2LimitDegrees=20, twistLimitDegrees=3}
 {bone1Name=dyn_strand_01_l, bone2Name=head, swing1LimitDegrees=10, swing2LimitDegrees=10, twistLimitDegrees=3}
 {bone1Name=dyn_strand_01_r, bone2Name=head, swing1LimitDegrees=15, swing2LimitDegrees=10, twistLimitDegrees=3}
]
